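; Sword decal actor for the ToyFox intro, holding a decal component that fades away on request.
; A custom event, FadeOutDecalAndDestroyActor, calls SetFadeOut on the decal so it fades after a short delay and then destroys the owning actor.
; The construction script is empty, and nothing in this graph triggers the event.
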(event Custom|FadeOutDecalAndDestroyActor
  ;[0.02093 0.03745 0.15 0.5] Fade out decal.
  (Rendering|Components|Decal|SetFadeOut (Variables|Default|GetDecal) 0.5 4.0 true))

(fn ConstructionScript)
