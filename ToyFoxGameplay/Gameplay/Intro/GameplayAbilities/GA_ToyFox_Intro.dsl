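; Gameplay ability that runs the ToyFox giant pawn's intro sky drop, from fall through landing to rise.
; Intro montage notifies switch intro collision, damage and destruction-trace state, toggle cues, and spawn the pedestal on impact.
; While falling, next-tick targeting repeatedly knocks back targets and destroys buildings; landing runs one async targeting pass at the event location.
; A sequencer-timed gameplay event triggers the Rise section and clears intro state.
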(event Ability|EventActivateAbilityFromEvent
  (params (EventData :type Gameplay%20Event%20Data%20Structure%20%28by%20ref%29))
  (bind _gameplaycuefall (Variables|GameplayCue|GetGameplayCueFall))
  (bind _gameplaycuetag (Utilities|Struct|BreakGameplayCueTag _gameplaycuefall))
  ;[0 0.15 0.006173 0.5] ActivateAbility: The pawn has just spawned in the sky. Start playing its falling animation while checking for landscape underneath.
  (|CallParentFunction EventData)
  (Fortnite|FortShippingLog self "ActivateAbility: Invoked.")
  (bind _asb_toy_fox_giant_player_pawn (Utilities|Casting|CastToB_ToyFoxGiantPlayerPawn (CallFunction|GetAvatarasFortPlayerPawn)))
  (Variables|Transient|SetAvatarGiantPlayerPawn _asb_toy_fox_giant_player_pawn)
  ;;[0.01044 0.04067 0.15 0.5] Start playing the falling montage.
  (bind (_asynctaskproxy _notifytag) (Ability|Tasks|PlayMontageAndWaitNotify "IntroMontage" (Variables|Settings|GetMontageIntro) 1.0 "None" false)
    (:then
      (Variables|Transient|AsyncTask|SetAsyncTask_IntroMontage _asynctaskproxy)
      (Utilities|FlowControl|Sequence
        (:then_0
          (Utilities|FlowControl|Sequence
            (:then_0
              ;;[0 0.15 0.13 0.5] Set up the montage start position to the desired value.
              (Animation|Montage|MontageSetPosition (Components|SkeletalMesh|GetAnimInstance (Variables|/Script/Engine.Character|GetMesh (Variables|Transient|GetAvatarGiantPlayerPawn))) (Animation|GetCurrentMontage (Variables|Transient|GetAvatarGiantPlayerPawn) :!class /Script/Engine.Character) (Math|Float|Clamp(Float) (ScalableFloat|GetValueatLevel (Variables|Hotfix|GetHotfix_MontageStartPosition)) 0.0 (Animation|GetPlayLength (Animation|GetCurrentMontage (Variables|Transient|GetAvatarGiantPlayerPawn) :!class /Script/Engine.Character))))
              (bind _returnvalue (Ability|ApplyGameplayEffectToOwner (Variables|Settings|GetIntroActiveGameplayEffectClass)))
              (Variables|Transient|SetIntroActiveGameplayEffectHandle _returnvalue)
              ;;[0.1306 8.3e-05 0.15 0.5] This event is initiated from the cinematic sequencer at the desired time.
              (Ability|Tasks|WaitGameplayEvent "(TagName=\"ToyFox.Event.Ability.Intro.Kneel.Rise\")" 0 true
                (:EventReceived
                  (Ability|Animation|MontageJumptoSection "Rise")
                  (Fortnite|FortShippingLog self "ActivateAbility: Rise montage section invoked.")
                  (CallFunction|SetIntroStateEnabled false))))
            (:then_1
              ;;[0.009784 0.086 0.15 0.5] Apply knockback immunity.
              (bind _returnvalue_2 (Ability|ApplyGameplayEffectToOwner (Variables|Settings|GetKnockbackImmuneGameplayEffectClass)))
              (Variables|Transient|SetKnockbackImmuneGEHandle _returnvalue_2)
              ;;[0.006819 0.02443 0.15 0.5] Force the animation update for the starting portion of the montage where the skeletal mesh is animated very far away from the actor location. This ensures the animation updates correctly on the clients.
              (Ability|AddGameplayCueToOwner (Utilities|Struct|BreakGameplayCueTag (Variables|GameplayCue|GetGameplayCueForceAnimUpdate)))
              ;;[0.15 0.05082 0.07783 0.5] Disable character movement so it doesn't use generic character movement after landing.
              (bind _outfortplayerpawn (CallFunction|GetAvatarasValidFortPlayerPawn)
                (:ValidFortPlayerPawn
                  (Pawn|Components|CharacterMovement|SetMovementMode (Variables|/Script/Engine.Character|GetCharacterMovement _outfortplayerpawn) "MOVE_None" 0) ; Without this, character doesn't remain exactly where it is spawned.
                  (Fortnite|FortShippingLog self "ActivateAbility: Successfully disabling movement."))))
            (:then_2
              ;;[0.04288 0.283 0.8177 0.5] Spawn actor for screen haze post-process.
              (Ability|HasAuthority
                (:True
                  (Game|SpawnActorfromClass "/ToyFoxGameplay/Assets/FX/B_ToyFox_UpScreenHaze.B_ToyFox_UpScreenHaze_C" (Math|Transform|MakeTransform (Transformation|GetSocketLocation (Variables|/Script/Engine.Character|GetMesh (Variables|Transient|GetAvatarGiantPlayerPawn)) "root")) "AlwaysSpawn" "MultiplyWithRoot" 0 0))))))
        (:then_1
          ;; Cache the event location actor so it can be referenced later. This is where the intro is occuring.
          (bind _returnvalue_3 (Fortnite|Athena|GameState|GetGameStateAthena))
          (Utilities|IsValid _returnvalue_3
            (:Is%20Valid
              (bind _returnvalue_4 (Actor|GetComponentbyClass _returnvalue_3 "/ToyFoxGameplay/Gameplay/B_ToyFoxGameStateComponent.B_ToyFoxGameStateComponent_C"))
              (Utilities|IsValid _returnvalue_4
                (:Is%20Valid
                  (bind _output_get (Variables|Transient|SetEventLocationActor (Event|GetSelectedEventLocationActor _returnvalue_4)))
                  (Fortnite|FortShippingLog self (Utilities|String|Append "ActivateAbility: Selected event location actor: " (Class|BToyFoxEventLocation|GetEventLocationName _output_get))))))))))
    (:OnCompleted
      (Ability|EndAbility))
    (:OnNotifyBegin
      (switch Utilities|FlowControl|Switch|SwitchonGameplayTag _notifytag
        (:ToyFox.AnimNotify.Ability.Intro.Fall.Start
          ;;[0.1224 0.03842 0.15 0.5] Character is falling. Start tracing for obstructions when falling.
          (Ability|AddGameplayCueToOwner _gameplaycuetag)
          (CallFunction|PerformObstructionTargeting)
          (CallFunction|SetIntroStateEnabled true))
        (:ToyFox.AnimNotify.Ability.Intro.Land.Impact
          ;;[0.01105 0.04477 0.15 0.5] Character has landed. Spawn pedestal and perform landed targeting.
          (Ability|RemoveGameplayCueFromOwner _gameplaycuetag)
          (Ability|RemoveGameplayCueFromOwner (Utilities|Struct|BreakGameplayCueTag (Variables|GameplayCue|GetGameplayCueForceAnimUpdate)))
          (CallFunction|StopObstructionTargeting)
          (bind (_bvalid _location) (Class|BToyFoxEventLocation|GetPlayerLocationandRotation (Variables|Transient|GetEventLocationActor)))
          (Ability|ExecuteGameplayCueWithParamsOnOwner (Utilities|Struct|BreakGameplayCueTag (Variables|GameplayCue|GetGameplayCueLand)) (GameplayCue|MakeGameplayCueParametersfromHitResult (Collision|MakeHitResult true false 0.0 0.0 "0, 0, 0" _location)))
          (CallFunction|SpawnPedestal)
          (CallFunction|PerformLandedTargeting))))))

(event Custom|PerformObstructionTargeting
  ;[0.1146 0 0.15 0.5] PerformObstructionTargeting: Destroy all building actors in the way and knockback vehicles and players in the way.
  (bind (_asynctaskproxy _targetdata) (Ability|Tasks|PerformTargetingAbilityTask "ObstructionTargeting" (Variables|Settings|GetObstructionTargetingPreset) false false (> (ScalableFloat|GetCachedScalableFloatValue (Variables|Hotfix|GetHotfix_ObstructionTargetingAsync)) 0))
    (:then
      (Variables|Transient|AsyncTask|SetAsyncTask_ObstructionTargeting _asynctaskproxy))
    (:Targeted
      _)
    (:Cancelled
      _))
  (Utilities|FlowControl|Sequence
    (:then_0
      (bind _returnvalue (Utilities|Time|SetTimerforNextTickbyEvent (delegate Custom|PerformObstructionTargeting)))
      (Variables|Transient|SetObstructionTargetingTimerHandle _returnvalue))
    (:then_1
      (Ability|ApplyGameplayEffectToTarget _targetdata (Variables|Settings|GetKnockbackFallingGameplayEffectClass))
      ;;[0.15 0.03432 0.1248 0.5] Destroying all buildings at once without any optimizations since there's not expected to be many buildings around the spawning area. The event starts early enough in the match that players are not going to have enough resources to build large structures in the area either.
      (for _array_element (Ability|TargetData|GetAllActorsfromTargetData _targetdata)
        (bind _asbuilding_actor (Utilities|Casting|CastToBuildingActor _array_element))
        (Gameplay|Die _asbuilding_actor (Variables|Settings|GetSilentDeathTag) 0 0 :!class /Script/FortniteGame.BuildingActor)))))

(event Custom|StopObstructionTargeting
  ;[0.1146 0 0.15 0.5] PerformObstructionTargeting: Destroy all building actors in the way and knockback vehicles and players in the way.
  (bind _asynctask_obstructiontargeting (Variables|Transient|AsyncTask|GetAsyncTask_ObstructionTargeting)
    (:then
      (GameplayTasks|EndTask _asynctask_obstructiontargeting))
    (:else
      _))
  (Utilities|Time|ClearandInvalidateTimerbyHandle (Variables|Transient|GetObstructionTargetingTimerHandle)))

(event Custom|PerformLandedTargeting
  ;[0.02734 0.08624 0.15 0.5] PerformLandedTargeting: The mech has landed. Check to see if there are nearby actors to influence.
  (bind (_bvalid _location) (Class|BToyFoxEventLocation|GetPlayerLocationandRotation (Variables|Transient|GetEventLocationActor)))
  (bind _returnvalue (TargetingSystem|AsyncRequest|StartAsyncTargetingRequest (GameInstanceSubsystems|GetTargetingSubsystem) (Variables|Settings|GetLandingKnockbackTargetingPreset) (Utilities|Struct|MakeTargetingSourceContext 0 (CallFunction|GetAvatarasFortPlayerPawn) (+ _location "0.000000,0.000000,20.000000")) (delegate Custom|HandleLandedTargetingComplete)))
  (Variables|Transient|SetTargetingRequestHandle _returnvalue))

(event Custom|HandleLandedTargetingComplete
  (params (TargetingRequestHandle :type Targeting%20Request%20Handle%20Structure))
  (bind _returnvalue (Ability|GetActivatingAbilityComponent))
  ;[0.02734 0.08624 0.15 0.5] PerformLandedTargeting: The mech has landed. Check to see if there are nearby actors to influence.
  (Utilities|FlowControl|Sequence
    (:then_0
      (Utilities|IsValid _returnvalue
        (:Is%20Valid
          (Ability|Tasks|ApplyGameplayEffectToTargetingHandleAsync _returnvalue TargetingRequestHandle (Variables|Settings|GetKnockbackLandedGameplayEffectClass) 1 1 (GameplayEffects|MakeEffectContext _returnvalue) (Variables|Hotfix|GetHotfix_MaxTargetsToProcessPerTick)))))
    (:then_1
      ;;[0.1134 0.03056 0.15 0.5] Destroying all buildings at once without any optimizations since there's not expected to be many buildings around the spawning area. The event starts early enough in the match that players are not going to have enough resources to build large structures in the area either.
      (for _array_element (TargetingSystem|TargetingResults|GetTargetingResults (GameInstanceSubsystems|GetTargetingSubsystem) TargetingRequestHandle)
        (bind _asbuilding_actor (Utilities|Casting|CastToBuildingActor (.HitActor (Collision|BreakHitResult _array_element))))
        (if (Variables|/Script/Engine.Actor|GetbCanBeDamaged _asbuilding_actor)
          (Gameplay|Die _asbuilding_actor (Variables|Settings|GetSilentDeathTag) 0 0 :!class /Script/FortniteGame.BuildingActor))))))

(event Ability|EventOnEndAbility
  (params (bWasCancelled :type Boolean))
  ;[0.125 0.06572 0 0.5] EndAbility: Clean up.
  (Utilities|FlowControl|Sequence
    (:then_0
      (GameplayTasks|EndTask (Variables|Transient|AsyncTask|GetAsyncTask_IntroMontage))
      (Ability|RemoveGameplayEffectFromOwnerWithHandle (Variables|Transient|GetIntroActiveGameplayEffectHandle))
      (TargetingSystem|AsyncRequest|RemoveAsyncTargetingRequestwithHandle (GameInstanceSubsystems|GetTargetingSubsystem) (Variables|Transient|GetTargetingRequestHandle))
      (CallFunction|SetIntroStateEnabled false)
      (Ability|RemoveGameplayEffectFromOwnerWithHandle (Variables|Transient|GetKnockbackImmuneGEHandle)))
    (:then_1
      (|CallParentFunction bWasCancelled))))

(fn CanActivateAbility
  (params (ActorInfo :type Gameplay%20Ability%20Actor%20Info%20Structure) (Handle :type Gameplay%20Ability%20Spec%20Handle%20Structure))
  (return 0 true))

(fn SetIntroStateEnabled
  (params (bEnabled :type Boolean))
  ;[0 0.15 0.06448 0.5] In the intro state, the pawn should block collision with players/vehicles, it should not be damageable, and it shouldn't be running destruction traces.
  (bind _avatargiantplayerpawn (Variables|Transient|GetAvatarGiantPlayerPawn)
    (:then
      (Intro|SetIntroCollisionStateEnabled _avatargiantplayerpawn bEnabled)
      (Fortnite|SetCanbeDamaged _avatargiantplayerpawn (not bEnabled))
      (bind _toyfoxgiantpawndestructioncomponent (Variables|/ToyFoxGameplay/Gameplay/B_ToyFoxGiantPlayerPawn.B_ToyFoxGiantPlayerPawn_C|GetToyFoxGiantPawnDestructionComponent _avatargiantplayerpawn)
        (:then
          (if bEnabled
            (Class|BGiantPawnDestructionComponent|StopSkeletonDestructionTraces _toyfoxgiantpawndestructioncomponent)
            (else
              (Class|BGiantPawnDestructionComponent|StartSkeletonDestructionTraces _toyfoxgiantpawndestructioncomponent))))
        (:else)))
    (:else)))

(fn SpawnPedestal
  (bind _returnvalue (Fortnite|Athena|GameState|GetGameStateAthena))
  (bind _returnvalue_2 (Actor|GetComponentbyClass _returnvalue "/ToyFoxGameplay/Gameplay/B_ToyFoxGameStateComponent.B_ToyFoxGameStateComponent_C"))
  ;[0.1213 0 0.15 0.5] SpawnPedestal: Let the game state component know that it is time to spawn the pedestal.
  (Utilities|IsValid _returnvalue
    (:Is%20Valid
      (Utilities|IsValid _returnvalue_2
        (:Is%20Valid
          (Class|BToyFoxGameStateComponent|SpawnToyFoxPedestalatEventLocation _returnvalue_2))))))
